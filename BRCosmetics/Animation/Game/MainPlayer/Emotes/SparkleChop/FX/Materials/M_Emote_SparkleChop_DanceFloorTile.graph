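# Lit material for the glowing dance floor tiles in the SparkleChop emote, built for Niagara mesh particles.
# A rounded-square function supplies tile height and a normal, which is flattened fully except on upward-facing surfaces.
# Emissive multiplies that height, particle color and a sphere-mask glow whose UVs shift with tangent-space view direction for fake parallax depth.
# A static switch can black out non-upward faces before the shared FX emissive adjustments.
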Material /BRCosmetics/Animation/Game/MainPlayer/Emotes/SparkleChop/FX/Materials/M_Emote_SparkleChop_DanceFloorTile
graph {
 properties {bAutomaticallySetUsageInEditor=false, bUsedWithNiagaraMeshParticles=true, usedShadingModels=Default Lit}
 nodes [
  n1 = MaterialFunctionCall!(UV, Size, Curvature=ScalarParameter() {defaultValue=0.1, parameterName=Square Curvature}) {materialFunction=/Game/Effects/Fort_Effects/MaterialFunctions/MF_RoundedSquare}
  n2 = Saturate(None=ComponentMask(None=VertexNormalWS()) {b=true})
  MaterialFunctionCall(
   Normal=@n1.Clamped Normal
   Flatness=LinearInterpolate(B=ScalarParameter() {defaultValue=0.333, parameterName=Normal Flatness Top}, Alpha=@n2) {constA=1}
  ) {materialFunction=&FlattenNormal}
  n3 = Subtract(A=CameraPositionWS(), B=WorldPosition() {worldPositionShaderOffset=WPT_ExcludeAllShaderOffsets}.XYZ)
  n4 = ComponentMask(None=Transform(None=Normalize(VectorInput=@n3)) {transformSourceType=TRANSFORMSOURCE_World, transformType=TRANSFORM_Tangent}) {g=true, r=true}
  n5 = ScalarParameter() {defaultValue=0.333, parameterName=Parallax Intensity}
  n6 = SphereMask(A=Subtract(A=TextureCoordinate(), B=Multiply(A=@n4, B=@n5)), B=Constant() {r=0.5}, Radius=Constant() {r=2}, Hardness=Constant())
  n7 = ScalarParameter() {defaultValue=6, parameterName=Falloff}
  n8 = Multiply(A=@n1.Height, B=Multiply(A=ParticleColor().RGB, B=Power(Base=@n6, Exp=@n7)))
  MaterialFunctionCall(Emissive=StaticSwitchParameter(True=Multiply(A=@n8, B=@n2), False=@n8) {parameterName=MakeBottomBlack}) {materialFunction=&MF_FX_EmissiveAdjustments}
  ScalarParameter() {defaultValue=0.2, parameterName=Specular}
  VectorParameter() {parameterName=Color, defaultValue=<rgba 0.01 0.01 0.01 0.7>}
 ]
}
