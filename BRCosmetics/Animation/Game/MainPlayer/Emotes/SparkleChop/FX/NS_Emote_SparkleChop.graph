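# Niagara system for the SparkleChop emote, staging a dance-club light show with a glowing tile floor, spotlights, lasers and sparkles.
# System Beat and Bar values index user color and position arrays, switching spotlight and idle laser looks in musical steps.
# LaserState gates idle lasers, while LeftLasers and RightLasers fade directional beams that sweep along timeline-driven curves.
# The main spotlight adds light and component renderers, with component intensity scaled by time of day.
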NiagaraSystem /BRCosmetics/Animation/Game/MainPlayer/Emotes/SparkleChop/FX/NS_Emote_SparkleChop
dependencies {
 usedDataInterfaces [
  &NiagaraDataInterfaceColorCurve
  &NiagaraDataInterfaceCurve
  &NiagaraDataInterfaceArrayFloat
  &NiagaraDataInterfaceVectorCurve
 ]
 usedDynamicInputs [
  &SimulationPosition
  &UniformRangedFloat
  &ModuloInt
 ]
 usedModules [
  &SystemState
  &Timeline
  &self:SystemSpawnScript.SystemScriptSource.SystemScriptGraph.NiagaraNodeAssignment_2.SetVariables_#1
  &self:SystemSpawnScript.SystemScriptSource.SystemScriptGraph.NiagaraNodeAssignment_3.SetVariables_#2
  &self:SystemSpawnScript.SystemScriptSource.SystemScriptGraph.NiagaraNodeAssignment_4.SetVariables_#3
 ]
}
emitters {
 FlashingSparkles {
  emitterUpdateScript [
   &EmitterState()
   &SpawnRate(SpawnRate=5, Use Spawn Probability=false)
  ]
  particleSpawnScript [
   &InitializeParticle(
    Color=&SelectLinearColorFromArray(Array Sampling Mode=::NewEnumerator0, Direct Array Index=?, Color Selection Array=$User.SpotlightColorArray, Random Seed=0)
    Lifetime Max=0.15
    Lifetime Min=0.2
    Lifetime Mode=::NewEnumerator1
    Sprite Rotation Angle=&UniformAOrBFloat(A=0, B=0.125)
    Sprite Rotation Mode=::NewEnumerator1
    Sprite Size Mode=::NewEnumerator1
    Uniform Sprite Size Max=66
    Uniform Sprite Size Min=33
    UsePositionOffset=false
    Write Material Random=true
   )
   &ShapeLocation(
    MaskSpawnGroup=false
    Non Uniform Scale=<0.666 0.666 1>
    Offset=<0 0 133>
    Offset Mode=::NewEnumerator0
    Shape Origin=<0 0 0>
    Sphere Distribution=::NewEnumerator1
    Sphere Radius=99
    Sphere Radius Position:NiagaraFloat=&Lerp_Float(A=0, B=1, Alpha=&Power(a=&RandomRangeFloat(Minimum=0, Maximum=1), b=1))
    U Position=&RandomRangeFloat(Minimum=0, Maximum=1)
    UseOwnerScale=true
    V Position=&RandomRangeFloat(Minimum=0, Maximum=1)
   )
   &CameraOffset(Camera Offset Amount=-99)
  ]
  particleUpdateScript [
   &ParticleState()
   &ScaleColor(
    Scale Alpha=&FloatFromCurve(FloatCurve=DataInterface() {propertyValues {Curve {keys [
         {arriveTangent=9.8157, interpMode=RCIM_Cubic, leaveTangent=9.8157, tangentMode=RCTM_User, time=0, value=0}
         {arriveTangent=-1.4617, interpMode=RCIM_Cubic, leaveTangent=-1.4617, tangentMode=RCTM_User, time=0.2, value=1}
         {interpMode=RCIM_Cubic, tangentMode=RCTM_User, time=0.3, value=0}
         {interpMode=RCIM_Cubic, tangentMode=RCTM_User, time=0.5, value=0.75}
         {interpMode=RCIM_Cubic, tangentMode=RCTM_Break, time=1, value=0}
        ]}}})
    ScaleA=true
    ScaleRGB=true
   )
   &ScaleSpriteSize(Uniform Curve Sprite Scale=DataInterface() {propertyValues {Curve {keys [{tangentMode=RCTM_User, time=0, value=1}, {tangentMode=RCTM_User, time=1, value=0}]}}})
  ]
  rendererClasses [&NiagaraSpriteRendererProperties]
  renderers [{rendererClass=&NiagaraSpriteRendererProperties}]
 }
 Laser_idle_01 {
  emitterUpdateScript [
   &EmitterState(
    Delay First Loop Only=true
    Inactive Response=::NewEnumerator1
    Life Cycle Mode=::NewEnumerator1
    Loop Delay=0.27
    Loop Duration=&Multiply_Float(A=$System.CurrentLoopDuration, B=2)
    UseLoopDelay=true
   )
   &SpawnBurst_Instantaneous(
    Spawn Count=6
    Spawn Probability=&MakeCustomFloatFromBool(
     Bool=&SetBoolByFloatComparison(A=$User.LaserState, B=0, Comparison Type=::NewEnumerator4)
     True Float=1
     False Float=0
    )
    Use Spawn Probability=true
    UseLoopCountLimit=false
   )
  ]
  particleSpawnScript [
   &PartitionParticles(Number Of Partitions=6)
   &InitializeParticle(
    Lifetime Max=0.75
    Lifetime Min=1.25
    Lifetime Mode=::NewEnumerator1
    Mesh Scale=&Multiply_Vector(
     A=&SelectVectorFromArray(
      Direct Array Index=$Particles.PartitionParticles.Partition
      Direct Array Mode=::NewEnumerator1
      Vector Selection Array=DataInterface() {propertyValues {FloatData [<0.1 0.1 1.2>, <0.2 0.2 1.2>, <0.1 0.1 1.2>, <0.15 0.15 1.2>, <0.1 0.1 1.2>, <0.175 0.175 1.2>]}}
     )
     B=<4 4 1.2>
    )
    Mesh Scale Mode=::NewEnumerator4
    Sprite Size Mode=::NewEnumerator3
    Uniform Sprite Size=100
    UsePositionOffset=false
    Write Material Random=true
   )
   &InitialMeshOrientation(
    Mesh Orientation Mode=::NewEnumerator2
    Rotation=&SelectVectorFromArray(Direct Array Index=$Particles.PartitionParticles.Partition, Vector Selection Array=DataInterface() {propertyValues {FloatData []}})
    Use Rotation Vector=true
   )
  ]
  particleUpdateScript [
   &ParticleState()
   /SetVariables(Particles.Position=&SelectVectorFromArray(
     Array Sampling Mode=::NewEnumerator0
     Direct Array Index=&Add_Integer(
      A=$Particles.PartitionParticles.Partition
      B=&Add_Integer(A=$System.Beat, B=&RandomRangeInt(Minimum=0, Maximum=23))
     )
     Direct Array Mode=::NewEnumerator1
     Vector Selection Array=$User.LasersPositionArray
     Random Seed=0
    )) {isSetParameters=true, module=SetVariables_#4}
   &Color(Color=&SelectLinearColorFromArray(
     Array Sampling Mode=::NewEnumerator1
     Direct Array Index=&Add_Integer(A=$System.Beat, B=&RandomRangeInt(Minimum=0, Maximum=3))
     Direct Array Mode=::NewEnumerator1
     Color Selection Array=$User.LaserColorArray
     Random Seed=?
    )) {module=Color001}
   &ScaleColor(
    Color Value To Scale=$Particles.Color
    Scale Alpha=&MakeCustomFloatFromBool(
     Bool=&SetBoolByFloatComparison(A=$User.LaserState, B=0, Comparison Type=::NewEnumerator7)
     True Float=1
     False Float=0
    )
    ScaleA=true
    ScaleRGB=false
   ) {module=ScaleColor001}
   &ScaleColor(
    Color Value To Scale=$Particles.Color
    Scale Alpha=&FloatFromCurve(
     FloatCurve=DataInterface() {propertyValues {Curve {keys [
         {arriveTangent=5.6401, interpMode=RCIM_Cubic, leaveTangent=5.6401, tangentMode=RCTM_User, time=0, value=0}
         {arriveTangent=0.0567, interpMode=RCIM_Cubic, leaveTangent=0.0567, tangentMode=RCTM_User, time=0.2042, value=0.6228}
         {arriveTangent=0.0224, interpMode=RCIM_Cubic, leaveTangent=0.0224, tangentMode=RCTM_User, time=1, value=0}
        ]}}}
     Scale Curve=0.05
    )
    Scale RGB=&VectorFromFloat(Value=&FloatFromCurve(FloatCurve=DataInterface() {propertyValues {Curve {keys [<at 0 0>,<at 0.2 10>,<at 0.3 0.5>,<at 0.4 12>,<at 0.6 2>,<at 0.7 8.5>,<at 1 0>]}}}))
    ScaleA=true
    ScaleRGB=true
   )
   &ScaleMeshSize(Scale Factor=&MakeVector(X=1, Y=1, Z=1))
   &UpdateMeshOrientation(
    Facing Direction=&NormalizeVector(
     Vector To Normalize=&AddVectorToPosition(
      Position=&Subtract_Vector(
       A=&ConvertPositionToVector(Input Position=$Particles.Position, Conversion Operation:ENiagara_LWCConvertPosToVec=::NewEnumerator0)
       B=&AddVectorToPosition(Position=$Particles.Position, Vector=&RandomRangeVector(Minimum=<-25 -25 111>, Maximum=<25 25 111>))
      )
      Vector=&SelectVectorFromArray(
       Direct Array Index=&Add_Integer(
        A=$Particles.PartitionParticles.Partition
        B=&Add_Integer(A=$System.Bar, B=&RandomRangeInt(Minimum=0, Maximum=3))
       )
       Direct Array Mode=::NewEnumerator1
       Vector Selection Array=DataInterface() {propertyValues {FloatData [<0 0 50>, <0 0 -50>, <0 0 50>, <0 0 -50>]}}
      )
     )
     Safe Normalize=true
     Safe Fallback Threshold=0
     Safe Fallback Vector=<0 0 0.001>
    )
    Facing Interpolation=false
    Orientation Method=::NewEnumerator1
    Override Quat=false
    Reference Axis Vector:Vector3f=<0 0 1>
    Stabilization Mode:ENiagara_OrientVectorCount=::NewEnumerator0
    TransformFacing=false
   ) {module=UpdateMeshOrientation002}
   &UpdateMeshOrientation(
    Rotation Coordinate Space=::NewEnumerator0
    Rotation Rate=&RandomRangeFloat(
     Minimum=&RandomRangeFloat(Minimum=-0.2, Maximum=-0.4)
     Maximum=&RandomRangeFloat(Minimum=0.2, Maximum=0.4)
    )
    Rotation Vector=<0.2 0.2 1>
   )
   /SetVariables(
    Particles.Color2=&MultiplyLinearColor(
     Linear Color=$Particles.Color
     Scale Factor=&MakeFloat4FromVectorAndFloat("Vector (RGB)"=&VectorFromFloat(Value=1000), "Float (Alpha)"=1)
    )
    Particles.Scale2:Vector3f=&Multiply_Vector(A=$Particles.Scale, B=<0.03 0.02 1>)
    Particles.ScaleThin:Vector3f=&Multiply_Vector(
     A=&Multiply_Vector(
      A=&SelectVectorFromArray(
       Direct Array Index=$Particles.PartitionParticles.Partition
       Direct Array Mode=::NewEnumerator1
       Vector Selection Array=DataInterface() {propertyValues {FloatData [<0.02 0.02 1.4>, <0.01 0.01 1.6>, <0.02 0.02 1.5>, <0.01 0.01 2>, <0.02 0.02 1.8>, <0.01 0.01 1.7>]}}
      )
      B=<4 4 1.2>
     )
     B=<0.03 0.02 1>
    )
    Particles.Size2:Vector2f=&Multiply_Vector2D(A=$Particles.SpriteSize, B=<2 2>)
   ) {isSetParameters=true, module=SetVariables_#5}
  ]
  rendererClasses [&NiagaraSpriteRendererProperties, &NiagaraMeshRendererProperties]
  renderers [
   {rendererClass=&NiagaraSpriteRendererProperties}
   {rendererClass=&NiagaraMeshRendererProperties, rendererIndex=1}
  ]
 }
 Laser_pointingLeft {
  emitterUpdateScript [
   &EmitterState()
   &SpawnBurst_Instantaneous(Spawn Count=3, Use Spawn Probability=true, UseLoopCountLimit=false)
  ]
  particleSpawnScript [
   &PartitionParticles(Number Of Partitions=3)
   &InitializeParticle(
    Lifetime=0.5
    Mesh Scale=&SelectVectorFromArray(
     Direct Array Index=$Particles.PartitionParticles.Partition
     Vector Selection Array=DataInterface() {propertyValues {FloatData [<0.3 0.3 1.55>, <0.3 0.3 1.65>, <0.3 0.3 1.55>]}}
    )
    Mesh Scale Mode=::NewEnumerator4
    Sprite Size Mode=::NewEnumerator3
    Uniform Sprite Size=100
    UsePositionOffset=false
    Write Material Random=true
   )
   &InitialMeshOrientation(
    Mesh Orientation Mode=::NewEnumerator2
    Rotation=&SelectVectorFromArray(Direct Array Index=$Particles.PartitionParticles.Partition, Vector Selection Array=DataInterface() {propertyValues {FloatData []}})
    Use Rotation Vector=true
   )
  ]
  particleUpdateScript [
   &ParticleState()
   /SetVariables(Particles.Position=&SelectVectorFromArray(
     Direct Array Index=$Particles.PartitionParticles.Partition
     Vector Selection Array=DataInterface() {propertyValues {FloatData [<-50 0 280>, <-50 0 280>, <-50 0 280>]}}
    )) {isSetParameters=true, module=SetVariables_#4}
   &Color(Color=&SelectLinearColorFromArray(
     Array Sampling Mode=::NewEnumerator1
     Direct Array Index=&Add_Integer(A=$System.Beat, B=&RandomRangeInt(Minimum=0, Maximum=3))
     Direct Array Mode=::NewEnumerator1
     Color Selection Array=$User.LaserColorArray
     Random Seed=?
    )) {module=Color001}
   &ScaleColor(Color Value To Scale=$Particles.Color, Scale Alpha=$User.LeftLasers, ScaleA=true, ScaleRGB=false) {module=ScaleColor001}
   &ScaleColor(
    Color Value To Scale=$Particles.Color
    Scale Alpha=&FloatFromCurve(
     FloatCurve=DataInterface() {propertyValues {Curve {keys [
         {arriveTangent=5.6401, interpMode=RCIM_Cubic, leaveTangent=5.6401, tangentMode=RCTM_User, time=0, value=0}
         {arriveTangent=0.0567, interpMode=RCIM_Cubic, leaveTangent=0.0567, tangentMode=RCTM_User, time=0.2042, value=0.6228}
         {arriveTangent=0.0224, interpMode=RCIM_Cubic, leaveTangent=0.0224, tangentMode=RCTM_User, time=1, value=0}
        ]}}}
     Scale Curve=0.05
    )
    Scale RGB=&VectorFromFloat(Value=&FloatFromCurve(FloatCurve=DataInterface() {propertyValues {Curve {keys [<at 0 0>,<at 0.2 2>,<at 0.7 2>,<at 1 0>]}}}))
    ScaleA=true
    ScaleRGB=true
   )
   &ScaleMeshSize(Scale Factor=&MakeVector(X=1, Y=1, Z=1))
   &UpdateMeshOrientation(
    Facing Direction=&NormalizeVector(
     Vector To Normalize=&AddVectorToPosition(
      Position=&Subtract_Vector(
       A=&ConvertPositionToVector(Input Position=$Particles.Position, Conversion Operation:ENiagara_LWCConvertPosToVec=::NewEnumerator0)
       B=<0 0 111>
      )
      Vector=&VectorFromCurve(
       VectorCurve=DataInterface() {propertyValues {XCurve {keys [<at 0 50>,<at 1 -50>]}, YCurve {keys [<at 0 25>,<at 1 50>]}, ZCurve {keys [<at 0 1>,<at 1 0>]}}}
       CurveIndex=$System.Timeline.NormalizedTime
       Scale Curve=&SelectVectorFromArray(
        Direct Array Index=$Particles.PartitionParticles.Partition
        Vector Selection Array=DataInterface() {propertyValues {FloatData [<-1 1 1.5>, <-1.5 0 1>, <-1 -1 1>]}}
       )
      )
     )
     Safe Normalize=true
     Safe Fallback Threshold=0
     Safe Fallback Vector=<0 0 0.001>
    )
    Facing Interpolation=false
    Orientation Method=::NewEnumerator1
    Override Quat=false
    Reference Axis Vector:Vector3f=<0 0 1>
    Stabilization Mode:ENiagara_OrientVectorCount=::NewEnumerator0
    TransformFacing=false
   )
   /SetVariables(
    Particles.Color2=&MultiplyLinearColor(
     Linear Color=$Particles.Color
     Scale Factor=&MakeFloat4FromVectorAndFloat("Vector (RGB)"=&VectorFromFloat(Value=500), "Float (Alpha)"=1)
    )
    Particles.Scale2:Vector3f=&Multiply_Vector(A=$Particles.Scale, B=<0.03 0.02 1>)
    Particles.Size2:Vector2f=&Multiply_Vector2D(A=$Particles.SpriteSize, B=<2 2>)
   ) {isSetParameters=true, module=SetVariables_#5}
  ]
  rendererClasses [&NiagaraSpriteRendererProperties, &NiagaraMeshRendererProperties]
  renderers [
   {rendererClass=&NiagaraSpriteRendererProperties}
   {rendererClass=&NiagaraMeshRendererProperties, rendererIndex=1}
  ]
 }
 Laser_pointingRight {
  emitterUpdateScript [
   &EmitterState()
   &SpawnBurst_Instantaneous(Spawn Count=3, Use Spawn Probability=true, UseLoopCountLimit=false)
  ]
  particleSpawnScript [
   &PartitionParticles(Number Of Partitions=3)
   &InitializeParticle(
    Lifetime=0.5
    Mesh Scale=&SelectVectorFromArray(
     Direct Array Index=$Particles.PartitionParticles.Partition
     Vector Selection Array=DataInterface() {propertyValues {FloatData [<0.3 0.3 1.55>, <0.3 0.3 1.65>, <0.3 0.3 1.55>]}}
    )
    Mesh Scale Mode=::NewEnumerator4
    Sprite Size Mode=::NewEnumerator3
    Uniform Sprite Size=100
    UsePositionOffset=false
    Write Material Random=true
   )
   &InitialMeshOrientation(
    Mesh Orientation Mode=::NewEnumerator2
    Rotation=&SelectVectorFromArray(Direct Array Index=$Particles.PartitionParticles.Partition, Vector Selection Array=DataInterface() {propertyValues {FloatData []}})
    Use Rotation Vector=true
   )
  ]
  particleUpdateScript [
   &ParticleState()
   /SetVariables(Particles.Position=&SelectVectorFromArray(
     Direct Array Index=$Particles.PartitionParticles.Partition
     Vector Selection Array=DataInterface() {propertyValues {FloatData [<50 0 280>, <50 0 280>, <50 0 280>]}}
    )) {isSetParameters=true, module=SetVariables_#4}
   &Color(Color=&SelectLinearColorFromArray(
     Array Sampling Mode=::NewEnumerator1
     Direct Array Index=&Add_Integer(A=$System.Beat, B=&RandomRangeInt(Minimum=0, Maximum=3))
     Direct Array Mode=::NewEnumerator1
     Color Selection Array=$User.LaserColorArray
     Random Seed=?
    )) {module=Color001}
   &ScaleColor(Color Value To Scale=$Particles.Color, Scale Alpha=$User.RightLasers, ScaleA=true, ScaleRGB=false) {module=ScaleColor001}
   &ScaleColor(
    Color Value To Scale=$Particles.Color
    Scale Alpha=&FloatFromCurve(
     FloatCurve=DataInterface() {propertyValues {Curve {keys [
         {arriveTangent=5.6401, interpMode=RCIM_Cubic, leaveTangent=5.6401, tangentMode=RCTM_User, time=0, value=0}
         {arriveTangent=0.0567, interpMode=RCIM_Cubic, leaveTangent=0.0567, tangentMode=RCTM_User, time=0.2042, value=0.6228}
         {arriveTangent=-0.0548, interpMode=RCIM_Cubic, leaveTangent=-0.0548, tangentMode=RCTM_User, time=1, value=0}
        ]}}}
     Scale Curve=0.05
    )
    Scale RGB=&VectorFromFloat(Value=&FloatFromCurve(FloatCurve=DataInterface() {propertyValues {Curve {keys [<at 0 0>,<at 0.2 2>,<at 0.7 2>,<at 1 0>]}}}))
    ScaleA=true
    ScaleRGB=true
   )
   &ScaleMeshSize(Scale Factor=&MakeVector(X=1, Y=1, Z=1))
   &UpdateMeshOrientation(
    Facing Direction=&NormalizeVector(
     Vector To Normalize=&AddVectorToPosition(
      Position=&Subtract_Vector(
       A=&ConvertPositionToVector(Input Position=$Particles.Position, Conversion Operation:ENiagara_LWCConvertPosToVec=::NewEnumerator0)
       B=<0 0 111>
      )
      Vector=&VectorFromCurve(
       VectorCurve=DataInterface() {propertyValues {XCurve {keys [<at 0 50>,<at 1 -50>]}, YCurve {keys [<at 0 25>,<at 1 50>]}, ZCurve {keys [<at 0 1>,<at 1 0>]}}}
       CurveIndex=$System.Timeline.NormalizedTime
       Scale Curve=&SelectVectorFromArray(
        Direct Array Index=$Particles.PartitionParticles.Partition
        Vector Selection Array=DataInterface() {propertyValues {FloatData [<1 1 1>, <1.5 0 1>, <1 -1 1.5>]}}
       )
      )
     )
     Safe Normalize=true
     Safe Fallback Threshold=0
     Safe Fallback Vector=<0 0 0.001>
    )
    Facing Interpolation=false
    Orientation Method=::NewEnumerator1
    Override Quat=false
    Reference Axis Vector:Vector3f=<0 0 1>
    Stabilization Mode:ENiagara_OrientVectorCount=::NewEnumerator0
    TransformFacing=false
   )
   /SetVariables(
    Particles.Color2=&MultiplyLinearColor(
     Linear Color=$Particles.Color
     Scale Factor=&MakeFloat4FromVectorAndFloat("Vector (RGB)"=&VectorFromFloat(Value=500), "Float (Alpha)"=1)
    )
    Particles.Scale2:Vector3f=&Multiply_Vector(A=$Particles.Scale, B=<0.03 0.02 1>)
    Particles.Size2:Vector2f=&Multiply_Vector2D(A=$Particles.SpriteSize, B=<2 2>)
   ) {isSetParameters=true, module=SetVariables_#5}
  ]
  rendererClasses [&NiagaraSpriteRendererProperties, &NiagaraMeshRendererProperties]
  renderers [
   {rendererClass=&NiagaraSpriteRendererProperties}
   {rendererClass=&NiagaraMeshRendererProperties, rendererIndex=1}
  ]
 }
 SoftSparkles {
  simTarget=GPUComputeSim
  emitterUpdateScript [
   &EmitterState()
   &SpawnRate(SpawnRate=40, Use Spawn Probability=false)
  ]
  particleSpawnScript [
   &InitializeParticle(
    Color=&MultiplyLinearColor(
     Linear Color=&SelectLinearColorFromArray(Array Sampling Mode=::NewEnumerator0, Direct Array Index=?, Color Selection Array=$User.SpotlightColorArray, Random Seed=0)
     Scale Factor=&MakeVector4_FromFloat(Float=0.1)
    )
    Lifetime Max=2.4
    Lifetime Min=1.2
    Lifetime Mode=::NewEnumerator1
    Sprite Size Mode=::NewEnumerator1
    Uniform Sprite Size Max=44
    Uniform Sprite Size Min=22
    UsePositionOffset=false
    Write Material Random=true
   )
   &ShapeLocation(
    MaskSpawnGroup=false
    Non Uniform Scale=<0.666 0.666 1>
    Offset=<0 0 133>
    Offset Mode=::NewEnumerator0
    Shape Origin=<0 0 0>
    Sphere Distribution=::NewEnumerator1
    Sphere Radius=99
    Sphere Radius Position:NiagaraFloat=&Lerp_Float(A=0, B=1, Alpha=&Power(a=&RandomRangeFloat(Minimum=0, Maximum=1), b=1))
    U Position=&RandomRangeFloat(Minimum=0, Maximum=1)
    UseOwnerScale=true
    V Position=&RandomRangeFloat(Minimum=0, Maximum=1)
   )
   /SetVariables(Particles.Curve:NiagaraFloat=&FloatFromCurve(FloatCurve=DataInterface() {propertyValues {Curve {keys [
         {interpMode=RCIM_Cubic, leaveTangent=7.0743, tangentMode=RCTM_Break, time=0, value=0.35}
         {interpMode=RCIM_Cubic, tangentMode=RCTM_Break, time=0.2, value=0.95}
         {interpMode=RCIM_Cubic, leaveTangent=-0.1241, tangentMode=RCTM_Break, time=0.4, value=0.55}
         {interpMode=RCIM_Cubic, tangentMode=RCTM_Break, time=0.6, value=0.8}
         {interpMode=RCIM_Cubic, tangentMode=RCTM_Break, time=0.8, value=0.6}
         {arriveTangent=2.6513, interpMode=RCIM_Cubic, tangentMode=RCTM_Break, time=1, value=1}
        ]}}})) {isSetParameters=true}
  ]
  particleUpdateScript [
   &ParticleState()
   &CurlNoiseForce(
    Noise Frequency=100
    Noise Strength=&FloatFromCurve(
     FloatCurve=DataInterface() {propertyValues {Curve {keys [{tangentMode=RCTM_User, time=0, value=0}, {tangentMode=RCTM_User, time=1, value=1}]}}}
     Scale Curve=111
    )
    Pan Noise=false
   )
   &SolveForcesAndVelocity(Clamp Velocity=false, Limit Acceleration=false)
   &DynamicMaterialParameters(
    Index 0 Param 1=&Multiply_Float(
     A=&FloatFromCurve(FloatCurve=DataInterface() {propertyValues {Curve {keys [
          {interpMode=RCIM_Cubic, leaveTangent=0.1514, tangentMode=RCTM_Break, time=0, value=1}
          {arriveTangent=-0.3559, interpMode=RCIM_Cubic, tangentMode=RCTM_Break, time=1, value=0}
         ]}}})
     B=&Lerp_Float(A=1, B=0, Alpha=$Particles.Curve)
    )
    Module.Param0WriteEnabled=true
    Module.Param1WriteEnabled=true
    Module.Param2WriteEnabled=true
    Module.Param3WriteEnabled=true
   )
   &ScaleColor(
    Scale Alpha=&FloatFromCurve(FloatCurve=DataInterface() {propertyValues {Curve {keys [
         {interpMode=RCIM_Cubic, tangentMode=RCTM_User, time=0, value=0}
         {interpMode=RCIM_Cubic, tangentMode=RCTM_User, time=0.5, value=1}
         {tangentMode=RCTM_User, time=1, value=0}
        ]}}})
    Scale RGB=&VectorFromFloat(Value=&Lerp_Float(
      A=1
      B=5
      Alpha=&Sine(Normalized Angle=$Emitter.Age, Period=&RandomRangeFloat(Minimum=0.1, Maximum=0.4), Scale=1, Bias=1)
     ))
    ScaleA=true
    ScaleRGB=true
   )
   &ScaleSpriteSize(
    Scale Sprite Size Mode=::NewEnumerator0
    Uniform Scale Factor=&Multiply_Float(
     A=&FloatFromCurve(FloatCurve=DataInterface() {propertyValues {Curve {keys [
          {interpMode=RCIM_Cubic, leaveTangent=2.1941, tangentMode=RCTM_Break, time=0, value=0.35}
          {arriveTangent=0.204, interpMode=RCIM_Cubic, tangentMode=RCTM_Break, time=1, value=1}
         ]}}})
     B=$Particles.Curve
    )
   )
  ]
  rendererClasses [&NiagaraSpriteRendererProperties]
  renderers [{rendererClass=&NiagaraSpriteRendererProperties}]
 }
 SpotLightMain {
  emitterUpdateScript [
   &EmitterState()
   &SpawnBurst_Instantaneous(Use Spawn Probability=false, UseLoopCountLimit=false)
  ]
  particleSpawnScript [
   &InitializeParticle(
    Lifetime=0.5
    Mesh Scale=<0.8 0.8 1.2>
    Mesh Scale Mode=::NewEnumerator4
    Sprite Size Mode=::NewEnumerator3
    Uniform Sprite Size=333
    UsePositionOffset=false
    Write Material Random=true
   )
   &CameraOffset()
   /SetVariables(
    Particles.FloorGlowFacingVector:Vector3f=<0 0 1>
    Particles.FloorGlowPosition:NiagaraPosition=&AddVectorToPosition(Position=/Niagara/DynamicInputs/Helpers/SimulationPosition(), Vector=<0 0 4>)
    Particles.FloorGlowSize:Vector2f=&Vector2DFromFloat(Value=250)
    Particles.LightAttenuationScale=275
    Particles.LightExponent=2
    Particles.LightPosition:NiagaraPosition=&AddVectorToPosition(Position=&SimulationPosition(), Vector=<0 0 200>)
    Particles.LightRadius=230
    Particles.SpotLightPosition:NiagaraPosition=&AddVectorToPosition(Position=&SimulationPosition(), Vector=<0 0 280>)
   ) {isSetParameters=true, module=SetVariables_#6}
  ]
  particleUpdateScript [
   &ParticleState()
   &Color(
    Color=&SelectLinearColorFromArray(
     Array Sampling Mode=::NewEnumerator1
     Direct Array Index=&Add_Integer(A=$System.Bar, B=&RandomRangeInt(Minimum=0, Maximum=3))
     Direct Array Mode=::NewEnumerator1
     Color Selection Array=$User.SpotlightColorArray
     Random Seed=?
    )
    Scale Alpha=0.2
   )
   /SetVariables(
    Particles.CoreSpotlightColor:LinearColor=&MultiplyLinearColor(
     Linear Color=<rgba 0.2 0.1687 0.16 1>
     Scale Factor=&MakeFloat4FromVectorAndFloat("Vector (RGB)"=<1 1 1>, "Float (Alpha)"=$System.Flicker)
    )
    Particles.FlareColor=&MultiplyLinearColor(Linear Color=$Particles.Color, Scale Factor=&MakeVector4_FromFloat(Float=0.333))
    Particles.FloorGlowColor:LinearColor=&MultiplyLinearColor(Linear Color=$Particles.Color, Scale Factor=&MakeVector4_FromFloat(Float=0.666))
    Particles.LightColor=&MultiplyLinearColor(Linear Color=$Particles.Color, Scale Factor=&MakeVector4_FromFloat(Float=8))
    Particles.LightComponentIntensity:NiagaraFloat=&Multiply_Float(
     A=$System.Oscillation
     B=&FNTimeOfDayScalar(Night Scale=5555.5, Morning Scale=11111, Day Scale=11111, Evening Scale=11111)
    )
   ) {isSetParameters=true}
  ]
  rendererClasses [
   &NiagaraMeshRendererProperties
   &NiagaraLightRendererProperties
   &NiagaraComponentRendererProperties
   &NiagaraSpriteRendererProperties
  ]
  renderers [
   {rendererClass=&NiagaraMeshRendererProperties, rendererIndex=0}
   {rendererClass=&NiagaraMeshRendererProperties, rendererIndex=1}
   {rendererClass=&NiagaraLightRendererProperties, rendererIndex=2}
   {rendererClass=&NiagaraComponentRendererProperties, rendererIndex=3}
   {rendererClass=&NiagaraMeshRendererProperties, rendererIndex=4}
   {rendererClass=&NiagaraMeshRendererProperties, rendererIndex=5}
   {rendererClass=&NiagaraSpriteRendererProperties, rendererIndex=6}
   {rendererClass=&NiagaraSpriteRendererProperties, rendererIndex=7}
  ]
 }
 SpotLightSecondary {
  emitterUpdateScript [
   &EmitterState()
   &SpawnBurst_Instantaneous(Spawn Count=2, Use Spawn Probability=false, UseLoopCountLimit=false)
  ]
  particleSpawnScript [
   &PartitionParticles()
   &InitializeParticle(
    Color=<rgba 0.3 0.271 0.24 1>
    Lifetime=0.5
    Mesh Scale=<0.8 0.8 1.2>
    Mesh Scale Mode=::NewEnumerator4
    Sprite Size Mode=::NewEnumerator3
    Uniform Sprite Size=111
    UsePositionOffset=false
    Write Material Random=true
   )
   &CameraOffset()
  ]
  particleUpdateScript [
   &ParticleState()
   /SetVariables(Particles.Position=&SelectVectorFromArray(
     Direct Array Index=&Add_Integer(A=$Particles.PartitionParticles.Partition, B=/Niagara/DynamicInputs/Divide/Divide_Int(A:NiagaraInt32=$System.Bar, B:NiagaraInt32=1))
     Direct Array Mode=::NewEnumerator1
     Vector Selection Array=$User.SpotlightsPositionArray
    )) {isSetParameters=true, module=SetVariables_#6}
   &UpdateMeshOrientation(
    Facing Direction=&NormalizeVector(
     Vector To Normalize=&Subtract_Vector(
      A=&ConvertPositionToVector(Input Position=$Particles.Position, Conversion Operation:ENiagara_LWCConvertPosToVec=::NewEnumerator0)
      B=<0 0 111>
     )
     Safe Normalize=true
     Safe Fallback Threshold=0
     Safe Fallback Vector=<0 0 0.001>
    )
    Facing Interpolation=false
    Orientation Method=::NewEnumerator1
    Override Quat=false
    Reference Axis Vector:Vector3f=<0 0 1>
    Stabilization Mode:ENiagara_OrientVectorCount=::NewEnumerator0
    TransformFacing=false
   )
   &Color(
    Color=&SelectLinearColorFromArray(
     Array Sampling Mode=::NewEnumerator1
     Direct Array Index=&Add_Integer(A=$Particles.PartitionParticles.Partition, B=&Multiply_Int(A=$System.Beat, B=2))
     Direct Array Mode=::NewEnumerator1
     Color Selection Array=$User.SpotlightColorArray
     Random Seed=?
    )
    Scale Alpha=&Multiply_Float(
     A=$System.Flicker
     B=&SelectFloatFromArray(
      Direct Array Index=&Add_Integer(A=&Multiply_Int(A=$System.Beat, B=3), B=$Particles.PartitionParticles.Partition)
      Direct Array Mode=::NewEnumerator1
      Float Selection Array=DataInterface() {propertyValues {FloatData [0.7,0.7,0.7,0.7]}}
     )
    )
    Scale Color=<0.4 0.4 0.4>
   )
   /SetVariables(
    Particles.GlowCameraOffset=11
    Particles.GlowColor=&MultiplyLinearColor(Linear Color=$Particles.Color, Scale Factor=&MakeVector4_FromFloat(Float=0.6))
   ) {isSetParameters=true}
  ]
  rendererClasses [&NiagaraMeshRendererProperties, &NiagaraSpriteRendererProperties]
  renderers [
   {rendererClass=&NiagaraMeshRendererProperties, rendererIndex=0}
   {rendererClass=&NiagaraMeshRendererProperties, rendererIndex=1}
   {rendererClass=&NiagaraSpriteRendererProperties, rendererIndex=2}
  ]
 }
 Tiles {
  emitterUpdateScript [
   &EmitterState()
   /Niagara/Modules/Spawn/Location/SpawnParticlesInGrid(X Count=7, Y Count=7, Z Count=0)
  ]
  particleSpawnScript [
   /Niagara/Modules/Spawn/Location/V2/GridLocation(
    Dimensions Definition:ENiagaraGridPlacementType=::NewEnumerator1
    Module.Jitter Inside Grid Cells=false
    Module.Override Random Seed=false
    Normalize Offsets=false
    Offset:Vector3f=<0 0 0.5>
    XYZ Dimensions:Vector3f=<95 95 0>
   ) {module=GridLocation001}
   &InitializeParticle(
    Color=&MakeCustomLinearColorFromBool(
     Bool=&SetBoolByFloatComparison(A=&RandomRangeFloat(Minimum=0, Maximum=1), B=0.7)
     True Color=<rgba 3 3 3 1>
     False Color=<rgba 0.0573 0.0573 0.0573 0>
    )
    Lifetime=0.5
    Mesh Scale=&MakeVector(
     X=&Divide_Float(
      A=0.95
      B=&MakeFloatFromInt(Int=/Niagara/DynamicInputs/Subtract/Subtract_Integer(A:NiagaraInt32=$Emitter.SpawnParticlesInGrid.Grid Location X Count, B:NiagaraInt32=1))
     )
     Y=&Divide_Float(
      A=0.95
      B=&MakeFloatFromInt(Int=&Subtract_Integer(A:NiagaraInt32=$Emitter.SpawnParticlesInGrid.Grid Location X Count, B:NiagaraInt32=1))
     )
     Z=0.01
    )
    Mesh Scale Mode=::NewEnumerator4
    Position Mode=::NewEnumerator0
    Write Material Random=true
   ) {module=InitializeParticle001}
   &self:ScaleSizeAndColorByDistance(
    Distance=55
    Inside Color:LinearColor=$Particles.Color
    Outside Color:LinearColor=<rgba 0 0 0 0>
    Outside Distance Scale:Vector3f=$Particles.Scale
   ) {module=ScaleSizeAndColorByDistance}
   /SetVariables(
    Particles.BeamSpriteCameraOffset=22
    Particles.BeamSpriteSize:Vector2f=<66 -66>
    Particles.GlowSpriteCameraOffset=33
    Particles.GlowSpriteSize:Vector2f=&Vector2DFromFloat(Value=50)
   ) {isSetParameters=true}
  ]
  particleUpdateScript [
   &ParticleState()
   &SolveForcesAndVelocity(Clamp Velocity=false, Limit Acceleration=false)
   &ScaleColor(
    Scale Alpha=&FloatFromCurve(FloatCurve=DataInterface() {propertyValues {Curve {keys [<at 0 0>,<at 0.1 1>,<at 0.9 1>,<at 1 0>]}}})
    Scale RGB=&Multiply_Vector(
     A=&MakeVectorFromLinearColorRGB(Color=&ColorFromCurve(
       ColorCurve=DataInterface() {propertyValues {
         AlphaCurve {keys [<at 0 1 RCIM_Cubic>]}
         BlueCurve {keys [<at 0.2 1 RCIM_Cubic>,<at 0.5 0.3>,<at 0.8 0.7801 RCIM_Cubic>]}
         GreenCurve {keys [<at 0.2 1 RCIM_Cubic>,<at 0.5 1>,<at 0.8 0.3 RCIM_Cubic>]}
         RedCurve {keys [<at 0.2 0.3 RCIM_Cubic>,<at 0.5 0.9417>,<at 0.8 1 RCIM_Cubic>]}
        }}
       CurveIndex=&RandomRangeFloat(Minimum=0, Maximum=1)
      ))
     B=&VectorFromFloat(Value=&FloatFromCurve(
       FloatCurve=DataInterface() {propertyValues {Curve {keys [{tangentMode=RCTM_User, time=0, value=0.2}, {tangentMode=RCTM_User, time=1, value=1}]}}}
       CurveIndex=&Multiply_Float(A=$Particles.NormalizedAge, B=&RandomRangeFloat(Minimum=0.1, Maximum=5))
      ))
    )
    ScaleA=true
    ScaleRGB=true
   )
   /SetVariables(
    Particles.BeamSpriteColor:LinearColor=&Multiply_LinearColorByFloat(Linear Color=$StackContext.Color, Float=0.11)
    Particles.BeamSpriteSize:Vector2f=&Multiply_Vector2D_ByFloat(
     Float=&MakeFloatFromLinearColor(LinearColor=$StackContext.Color, Channel:ENiagaraLinearColor_Channels=::NewEnumerator3)
     Vector2D=$Particles.Initial.BeamSpriteSize
    )
    Particles.GlowSpriteColor:LinearColor=&Multiply_LinearColorByFloat(Linear Color=$StackContext.Color, Float=0.22)
    Particles.GlowSpriteSize:Vector2f=&Multiply_Vector2D_ByFloat(
     Float=&MakeFloatFromLinearColor(LinearColor=$StackContext.Color, Channel:ENiagaraLinearColor_Channels=::NewEnumerator3)
     Vector2D=$Particles.Initial.GlowSpriteSize
    )
   ) {isSetParameters=true}
   &SpriteFacingAndAlignment(Sprite Alignment=<0 0 1>, WriteAlignment=true, WriteFacing=false)
  ]
  rendererClasses [&NiagaraMeshRendererProperties, &NiagaraSpriteRendererProperties]
  renderers [
   {rendererClass=&NiagaraMeshRendererProperties, rendererIndex=0}
   {rendererClass=&NiagaraSpriteRendererProperties, rendererIndex=1}
   {rendererClass=&NiagaraSpriteRendererProperties, rendererIndex=2}
  ]
 }
}
summary {
 systemName=NS_Emote_SparkleChop
 emitters [
  {emitterName=Tiles, rendererClasses [&NiagaraMeshRendererProperties, &NiagaraSpriteRendererProperties]}
  {
   emitterName=SpotLightMain
   rendererClasses [
    &NiagaraMeshRendererProperties
    &NiagaraLightRendererProperties
    &NiagaraComponentRendererProperties
    &NiagaraSpriteRendererProperties
   ]
  }
  {
   emitterName=SpotLightSecondary
   rendererClasses [&NiagaraMeshRendererProperties, &NiagaraSpriteRendererProperties]
  }
  {emitterName=Laser_idle_01, rendererClasses [&NiagaraSpriteRendererProperties, &NiagaraMeshRendererProperties]}
  {emitterName=SoftSparkles, simTarget=GPUComputeSim, rendererClasses [&NiagaraSpriteRendererProperties]}
  {emitterName=FlashingSparkles, rendererClasses [&NiagaraSpriteRendererProperties]}
  {
   emitterName=Laser_pointingLeft
   rendererClasses [&NiagaraSpriteRendererProperties, &NiagaraMeshRendererProperties]
  }
  {
   emitterName=Laser_pointingRight
   rendererClasses [&NiagaraSpriteRendererProperties, &NiagaraMeshRendererProperties]
  }
 ]
 userVariables [
  User.bIsFrontEnd {type=NiagaraBool}
  User.bIsFrontEndPreview {type=NiagaraBool}
  User.LaserColorArray {
   type=NiagaraDataInterfaceArrayColor
   description=""
   defaultValue {
    struct=NiagaraExt_VariableValue_DataInterface
    value {
     dataInterface=&self:NiagaraDataInterfaceArrayColor_0
     dataInterfaceClass=&NiagaraDataInterfaceArrayColor
    }
   }
  }
  User.LasersPositionArray {
   type=NiagaraDataInterfaceArrayFloat3
   description=""
   defaultValue {
    struct=NiagaraExt_VariableValue_DataInterface
    value {
     dataInterface=&self:NiagaraDataInterfaceArrayFloat3_0
     dataInterfaceClass=&NiagaraDataInterfaceArrayFloat3
    }
   }
  }
  User.LaserState {type=NiagaraFloat, defaultValue=0, description=""}
  User.LeftLasers {type=NiagaraFloat, defaultValue=0, description=""}
  User.RightLasers {type=NiagaraFloat, defaultValue=0, description=""}
  User.SongBPM {type=NiagaraFloat, defaultValue=120, description=""}
  User.SpotlightColorArray {
   type=NiagaraDataInterfaceArrayColor
   description=""
   defaultValue {
    struct=NiagaraExt_VariableValue_DataInterface
    value {
     dataInterface=&self:NiagaraDataInterfaceArrayColor_1
     dataInterfaceClass=&NiagaraDataInterfaceArrayColor
    }
   }
  }
  User.SpotlightsPositionArray {
   type=NiagaraDataInterfaceArrayFloat3
   description=""
   defaultValue {
    struct=NiagaraExt_VariableValue_DataInterface
    value {
     dataInterface=&self:NiagaraDataInterfaceArrayFloat3_1
     dataInterfaceClass=&NiagaraDataInterfaceArrayFloat3
    }
   }
  }
 ]
}
